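# Physics asset for the ProdigyHaughty backpack cosmetic, defining bodies and joints for three dangling chains on the pack.
# Capsule bodies cover the pack, three chain bases, a four-link center chain and two-link left and right chains.
# Bases join the pack with free swing and twist, and each link is constrained to the one before it.
# Link twist limits are near zero, while center-chain swing limits widen toward the tip.
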PhysicsAsset /BRCosmetics/Accessories/FORT_Backpacks/Backpack_M_MED_ProdigyHaughty/Meshes/M_MED_ProdigyHaughty_Pack_Physics
bodies {
 base_chain_c [{shapeType=Capsule, length=0.505, radius=0.505, shapeName=base_chain_c_capsule, rotation=<rot 90 0 0>}]
 base_chain_l [{shapeType=Capsule, length=0.505, radius=0.505, shapeName=base_chain_l_capsule, rotation=<rot 90 0 0>}]
 base_chain_r [{shapeType=Capsule, length=0.505, radius=0.505, shapeName=base_chain_r_capsule, rotation=<rot 90 0 0>}]
 dyn_chain_1_c [{shapeType=Capsule, length=0.375, radius=0.625, shapeName=dyn_chain_1_c_capsule, center=<1.5008 0 0>, rotation=<rot 90 0 0>}]
 dyn_chain_1_l [{shapeType=Capsule, length=0.8895, radius=0.777, shapeName=dyn_chain_1_l_capsule, center=<4.5 0 0>, rotation=<rot 90 0 0>}]
 dyn_chain_1_r [{shapeType=Capsule, length=1.1925, radius=1.005, shapeName=dyn_chain_1_r_capsule, center=<4.5 0 0>, rotation=<rot 90 0 0>}]
 dyn_chain_2_c [{shapeType=Capsule, length=0.375, radius=0.625, shapeName=dyn_chain_2_c_capsule, center=<0.7123 0 0>, rotation=<rot 90 0 0>}]
 dyn_chain_2_l [{shapeType=Capsule, length=0.7332, radius=0.5052, shapeName=dyn_chain_2_l_capsule, center=<4.5 0 0>, rotation=<rot 90 0 0>}]
 dyn_chain_2_r [{shapeType=Capsule, length=0.505, radius=0.8175, shapeName=dyn_chain_2_r_capsule, center=<4.5 0 0>, rotation=<rot 90 0 0>}]
 dyn_chain_3_c [{shapeType=Capsule, length=0.375, radius=0.625, shapeName=dyn_chain_3_c_capsule, center=<1.0762 0 0>, rotation=<rot 90 0 0>}]
 dyn_chain_4_c [{shapeType=Capsule, length=0.5, radius=1.625, shapeName=dyn_chain_4_c_capsule, center=<2.4539 0 0>, rotation=<rot 90 0 0>}]
 pack [{shapeType=Capsule, length=1, radius=1.25, center=<0.0373 0 0>, rotation=<rot 90 0 0>}]
}
constraints [
 {
  bone1Name=base_chain_c
  bone2Name=pack
  swing1LimitDegrees=45
  swing1Motion=Free
  swing2LimitDegrees=45
  swing2Motion=Free
  twistLimitDegrees=45
  twistMotion=Free
 }
 {
  bone1Name=base_chain_l
  bone2Name=pack
  swing1LimitDegrees=45
  swing1Motion=Free
  swing2LimitDegrees=45
  swing2Motion=Free
  twistLimitDegrees=45
  twistMotion=Free
 }
 {
  bone1Name=base_chain_r
  bone2Name=pack
  swing1LimitDegrees=45
  swing1Motion=Free
  swing2LimitDegrees=45
  swing2Motion=Free
  twistLimitDegrees=45
  twistMotion=Free
 }
 {bone1Name=dyn_chain_1_c, bone2Name=base_chain_c, swing1LimitDegrees=10, swing2LimitDegrees=10, twistLimitDegrees=1}
 {bone1Name=dyn_chain_2_c, bone2Name=dyn_chain_1_c, swing1LimitDegrees=15, swing2LimitDegrees=15, twistLimitDegrees=1}
 {bone1Name=dyn_chain_3_c, bone2Name=dyn_chain_2_c, swing1LimitDegrees=25, swing2LimitDegrees=25, twistLimitDegrees=1}
 {bone1Name=dyn_chain_4_c, bone2Name=dyn_chain_3_c, swing1LimitDegrees=35, swing2LimitDegrees=35, twistLimitDegrees=1}
 {bone1Name=dyn_chain_1_l, bone2Name=base_chain_l, swing1LimitDegrees=25, swing2LimitDegrees=10, twistLimitDegrees=0}
 {bone1Name=dyn_chain_2_l, bone2Name=dyn_chain_1_l, swing1LimitDegrees=8, swing2LimitDegrees=8, twistLimitDegrees=0}
 {bone1Name=dyn_chain_1_r, bone2Name=base_chain_r, swing1LimitDegrees=15, swing2LimitDegrees=10, twistLimitDegrees=0}
 {bone1Name=dyn_chain_2_r, bone2Name=dyn_chain_1_r, swing1LimitDegrees=15, swing2LimitDegrees=5, twistLimitDegrees=1}
]
